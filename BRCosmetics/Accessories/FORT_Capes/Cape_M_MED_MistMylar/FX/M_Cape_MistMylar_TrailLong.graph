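# Translucent, unlit, two-sided material for the long smoke trail on the MistMylar cape cosmetic.
# Emissive blends time-of-day gray smoke with a gold curve-atlas variant via UseGoldVersion.
# Opacity layers panning smoke and noise, some noise-distorted, with UV edge fades and depth fade.
# World position offset combines SplineThicken, sine waving and camera offsets, with vertex red varying strands, vertex green masking side offsets, and platform switches tuning width and wave intensity.
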Material /BRCosmetics/Accessories/FORT_Capes/Cape_M_MED_MistMylar/FX/M_Cape_MistMylar_TrailLong
graph {
 properties {
  bAutomaticallySetUsageInEditor=false
  bUsedWithMeshDeformer=true
  bUsedWithSkeletalMesh=true
  blendMode=BLEND_Translucent
  shadingModel=MSM_Unlit
  twoSided=true
  usedShadingModels=Unlit
 }
 nodes [
  n1 = VectorParameter() {parameterName=GrayColor, defaultValue=<rgba 0.5525 0.6305 0.65 1>}
  n2 = MaterialFunctionCall(
   Color_Morning=@n1.RGB
   Color_Day=@n1.RGB
   Color_Evening=@n1.RGB
   Color_Night=VectorParameter() {parameterName=NightColor, defaultValue=<rgba 0.255 0.3923 0.85 1>}.RGB
  ) {materialFunction=&Color_Time_of_Day}
  n3 = TextureCoordinate()
  n4 = AppendVector(A=VertexColor().R, B=Constant() {r=1})
  n5 = Multiply(A=AppendVector(A=Constant() {r=0.8}, B=Constant() {r=0.5}), B=@n4)
  n6 = AppendVector(A=Constant() {r=-0.3}, B=Constant())
  n7 = TextureSample(UVs=Panner(Coordinate=Multiply(A=@n3, B=@n5), Speed=Multiply(A=@n6, B=@n4)) {speedX=-0.3}) {texture=&T_Inky_Smoke_Tile}
  n8 = TextureCoordinate() {uTiling=1.2}
  n9 = Multiply(
   A=LinearInterpolate(Alpha=Multiply(A=@n7.R, B=@n7.R)) {constA=0.1, constB=0.6}
   B=TextureSample(UVs=Panner(Coordinate=@n8) {speedX=-0.2}) {samplerType=SAMPLERTYPE_Grayscale, texture=&T_Smoke_Trail_Soft}.R
  )
  n10 = TextureCoordinate()
  n11 = OneMinus(None=ComponentMask(None=@n10) {r=true})
  n12 = ComponentMask(None=Add(A=@n10, B=Constant2Vector() {r=0.05})) {r=true}
  n13 = Multiply(A=Power(Base=Multiply(A=@n9) {constB=0.6}) {constExponent=1}, B=Multiply(A=Multiply(A=@n11, B=@n12)) {constB=1.5})
  n14 = VectorParameter() {parameterName=GradationColor1, defaultValue=<rgba 0.5 0.2461 0.0239 1>}
  n15 = Add(A=TextureCoordinate(), B=NamedRerouteUsage()) {constB=0}
  n16 = AppendVector(A=VertexColor().R, B=Constant())
  n17 = TextureSample(UVs=Panner(Coordinate=Add(A=@n15, B=@n16)) {speedX=-0.5}) {texture=&T_Inky_Smoke_Tile}
  n18 = LinearInterpolate(Alpha=@n17.R) {constA=0.5}
  n19 = TextureCoordinate()
  n20 = TextureCoordinate() {uTiling=0.7}
  n21 = AppendVector(A=VertexColor().R, B=Constant())
  n22 = TextureSample(UVs=Panner(Coordinate=Add(A=@n20, B=@n21)) {speedX=-0.3}) {samplerType=SAMPLERTYPE_Grayscale, texture=&T_Smoke_Trail_Soft}
  n23 = ComponentMask(None=TextureCoordinate()) {r=true}
  n24 = Saturate(None=Multiply(A=Power(Base=@n23) {constExponent=1.8}) {constB=10})
  n25 = Multiply(
   A=LinearInterpolate(A=Multiply(A=@n14.RGB, B=@n18), B=@n14.RGB, Alpha=ComponentMask(None=@n19) {r=true})
   B=Power(Base=Multiply(A=@n22.R, B=@n24)) {constExponent=3}
  )
  n26 = CurveAtlasRowParameter(CurveTime=ComponentMask(None=TextureCoordinate()) {r=true}) {
   atlas=&/CAtlas_MistMylar_FX
   curve=&/CV_MistMylar_Color
   parameterName=Color
  }
  n27 = Multiply(A=Power(Base=@n17.R) {constExponent=0.5}, B=@n26)
  MaterialFunctionCall(Emissive=LinearInterpolate(A=Multiply(A=@n2.Result, B=@n13), B=Add(A=@n25, B=@n27), Alpha=ScalarParameter() {parameterName=UseGoldVersion})) {materialFunction=&MF_FX_EmissiveAdjustments}
  n28 = VertexColor()
  n29 = Constant()
  n30 = Panner(Coordinate=Add(A=TextureCoordinate() {uTiling=0.8}, B=AppendVector(A=@n28.R, B=@n29))) {speedX=0.1}
  Multiply(A=Subtract(A=TextureSample(UVs=@n30) {samplerType=SAMPLERTYPE_Grayscale, texture=&LowResBlurredNoise}.R) {constB=0.5}) {constB=0.2}
  n31 = TextureCoordinate() {uTiling=0.8}
  n32 = AppendVector(A=VertexColor().R, B=Constant())
  n33 = TextureSample(UVs=Panner(Coordinate=Add(A=@n31, B=@n32)) {speedX=0.1}) {samplerType=SAMPLERTYPE_Grayscale, texture=&LowResBlurredNoise}
  NamedRerouteDeclaration(None=Multiply(A=Subtract(A=@n33.R) {constB=0.5}) {constB=0.2}) {name=Disturb, nodeColor=<rgba 1 0.1641 0 1>}
  n34 = Constant() {r=0.1}
  n35 = Constant() {r=0.3}
  n36 = Constant() {r=2.5}
  n37 = Constant() {r=1.8}
  n38 = ComponentMask(None=TextureCoordinate()) {r=true}
  n39 = Power(Base=@n38) {constExponent=1.2}
  n40 = Multiply(A=OneMinus(None=@n38)) {constB=2}
  n41 = LinearInterpolate(
   A=MaterialFunctionCall("Xb1,PS4"=@n34, "PC-High&&Epic,XSX,PS5"=@n34, "PC-Perf,Android"=@n35, Switch=@n35, "PC-Low(NotPerf)&&Medium"=@n34) {materialFunction=&PlatformSwitch_1D}.Result
   B=MaterialFunctionCall("Xb1,PS4"=@n36, "PC-High&&Epic,XSX,PS5"=@n36, "PC-Perf,Android"=@n37, Switch=@n37, "PC-Low(NotPerf)&&Medium"=@n36) {materialFunction=&PlatformSwitch_1D}.Result
   Alpha=Multiply(A=LinearInterpolate(Alpha=@n39), B=Saturate(None=@n40)) {constA=1}
  ) {constA=0.1, constB=2.5}
  n42 = Constant() {r=15}
  n43 = Multiply(A=TextureCoordinate(), B=LinearInterpolate(Alpha=VertexColor().R) {constA=0.5, constB=1.2})
  n44 = Time()
  n45 = VertexColor()
  n46 = Add(A=Power(Base=ComponentMask(None=@n43) {r=true}) {constExponent=0.5}, B=Multiply(A=Add(A=@n44, B=@n45.R) {constB=0}) {constB=-0.5})
  n47 = Multiply(
   A=LinearInterpolate(Alpha=Sine(None=@n46) {period=0.7}) {constA=-0.5, constB=0.5}
   B=Transform(None=VectorParameter() {parameterName=WPODirection, defaultValue=<rgba 0 0 1 1>}.RGB)
  )
  n48 = TextureCoordinate()
  n49 = LinearInterpolate(Alpha=VertexColor().R) {constA=0.5, constB=1.2}
  n50 = Power(Base=ComponentMask(None=Multiply(A=@n48, B=@n49)) {r=true}) {constExponent=0.5}
  n51 = Multiply(A=Add(A=Time(), B=VertexColor().R) {constB=0}) {constB=-0.5}
  n52 = Multiply(
   A=Sine(None=Add(A=@n50, B=@n51) {constB=0}) {period=0.7}
   B=Transform(None=VectorParameter() {parameterName=WPODirection02, defaultValue=<rgba 0 0.6 0 1>}.RGB) {transformSourceType=TRANSFORMSOURCE_Local}
  )
  n53 = ScalarParameter() {defaultValue=7, parameterName=WPOIntensity}
  n54 = ScalarParameter() {defaultValue=5, parameterName=WPOIntensityLOD1}
  n55 = Multiply(
   A=Add(A=Multiply(A=@n47), B=Multiply(A=@n52)) {constB=0}
   B=MaterialFunctionCall("Xb1,PS4"=@n53, "PC-High&&Epic,XSX,PS5"=@n53, "PC-Perf,Android"=@n54, Switch=@n54, "PC-Low(NotPerf)&&Medium"=@n53) {materialFunction=&PlatformSwitch_1D}.Result
  )
  n56 = Power(Base=ComponentMask(None=TextureCoordinate()) {r=true}, Exp=ScalarParameter() {defaultValue=0.5, parameterName=TopMask}) {constExponent=0.5}
  n57 = TextureCoordinate()
  n58 = Power(Base=OneMinus(None=ComponentMask(None=@n57) {r=true})) {constExponent=0.5}
  n59 = Saturate(None=Multiply(A=LinearInterpolate(Alpha=@n58)) {constB=1.5})
  n60 = Saturate(None=Multiply(A=Multiply(A=@n56, B=@n59)) {constB=1.3})
  n61 = Add(
   A=MaterialFunctionCall(WidthBase=Multiply(A=@n41, B=@n42) {constA=1}, WidthTip=Multiply(A=@n41, B=@n42) {constA=1}, Expand U or V UV Channel=StaticBool()) {materialFunction=&SplineThicken}.WorldPositionOffset
   B=Add(A=Multiply(A=@n55, B=@n60)) {constB=0}
  ) {constB=0}
  n62 = VectorParameter() {parameterName=Offset02, defaultValue=<rgba 0 -2 2 1>}
  n63 = VectorParameter() {parameterName=SideMeshOffset, defaultValue=<rgba 0 -1 0 1>}
  n64 = Add(
   A=Add(A=Add(A=@n61) {constB=0}, B=Transform(None=@n62.RGB) {transformSourceType=TRANSFORMSOURCE_Local})
   B=Multiply(A=Transform(None=@n63.RGB) {transformSourceType=TRANSFORMSOURCE_Local}, B=VertexColor().G)
  )
  n65 = Multiply(
   A=VertexColor().G
   B=MaterialFunctionCall(Offset Amount=ScalarParameter() {defaultValue=3, parameterName="SideOffset Amount02 (S)"}) {materialFunction=/Engine/Functions/Engine_MaterialFunctions02/WorldPositionOffset/CameraOffset}.Camera Offset
  )
  n66 = ScalarParameter() {defaultValue=7, parameterName="Offset Amount (S)"}
  Reroute(None=Add(
    A=Add(A=@n64, B=@n65)
    B=MaterialFunctionCall(Offset Amount=@n66) {materialFunction=&CameraOffset}.Camera Offset
   ))
  n67 = Add(A=TextureCoordinate(), B=AppendVector(A=VertexColor().R, B=Constant()))
  n68 = NamedRerouteUsage()
  n69 = TextureSample(UVs=Panner(Coordinate=Add(A=@n67, B=@n68) {constB=0}) {speedX=-1}) {samplerType=SAMPLERTYPE_Grayscale, texture=&T_Smoke_Trail_Soft}
  n70 = ComponentMask(None=TextureCoordinate()) {g=true}
  n71 = LinearInterpolate(Alpha=OneMinus(None=Sine(None=@n70) {period=2})) {constA=0.2}
  n72 = Add(A=@n68, B=TextureCoordinate() {uTiling=3, vTiling=0.7})
  n73 = Constant2Vector() {g=0.25}
  n74 = TextureSample(UVs=Panner(Coordinate=Add(A=@n72, B=@n73)) {speedX=-1}) {texture=/BRCosmetics/Weapons/FORT_Melee/Pickaxe_Future_Samurai_Summer_Male/FX/T_Pickaxe_Future_Samurai_Summer_Male_Trail_01}
  n75 = Saturate(None=LinearInterpolate(Alpha=Power(Base=@n74.R) {constExponent=1}))
  n76 = Multiply(A=Multiply(A=Multiply(A=@n71, B=@n75))) {constB=1.5}
  n77 = VertexColor()
  n78 = Constant()
  n79 = NamedRerouteUsage()
  n80 = Add(A=Add(A=TextureCoordinate() {vTiling=0.7}, B=AppendVector(A=@n77.R, B=@n78)), B=@n79) {constB=0}
  n81 = VertexColor()
  n82 = Constant()
  n83 = Add(A=Add(A=TextureCoordinate() {vTiling=0.7}, B=AppendVector(A=@n81.R, B=@n82)), B=@n79) {constB=0}
  n84 = Add(
   A=TextureSample(UVs=Panner(Coordinate=@n80) {speedX=-0.5}) {samplerType=SAMPLERTYPE_Grayscale, texture=&T_Smoke_Trail_Soft}.R
   B=TextureSample(UVs=Panner(Coordinate=@n83) {speedX=-0.3}) {samplerType=SAMPLERTYPE_Grayscale, texture=&LowResBlurredNoise}.R
  )
  n85 = TextureCoordinate() {uTiling=1.5}
  n86 = AppendVector(A=VertexColor().R, B=Constant())
  n87 = TextureSample(UVs=Panner(Coordinate=Add(A=@n85, B=@n86)) {speedX=-1}) {samplerType=SAMPLERTYPE_Grayscale, texture=&T_Smoke_Trail_Soft}
  n88 = Power(Base=Saturate(None=Add(A=@n84, B=@n87.R) {constB=0})) {constExponent=1.5}
  n89 = LinearInterpolate(A=Multiply(A=Add(A=@n69.R, B=@n76) {constB=0}) {constB=1.2}, B=Add(A=Multiply(A=@n88) {constB=0.9}) {constB=0}) {constAlpha=0.8}
  n90 = Sine(None=ComponentMask(None=TextureCoordinate()) {g=true}) {period=2}
  n91 = OneMinus(None=ComponentMask(None=TextureCoordinate()) {r=true})
  n92 = Multiply(A=Multiply(A=Multiply(B=@n89) {constA=1}, B=Power(Base=@n90) {constExponent=1}), B=Saturate(None=Multiply(A=@n91) {constB=10}))
  n93 = ComponentMask(None=TextureCoordinate()) {r=true}
  n94 = Saturate(None=Multiply(A=Power(Base=@n93) {constExponent=1.2}, B=ScalarParameter() {defaultValue=10, parameterName=TopAlphaFade}) {constB=20})
  StaticSwitch(
   True=Constant() {r=1}
   False=DepthFade(Opacity=Multiply(A=@n92, B=@n94), FadeDistance=ScalarParameter() {defaultValue=8, parameterName=FadeDistance}) {fadeDistanceDefault=8}
  )
 ]
}
