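# Translucent unlit material for the top smoky trail effect on the MistMylar cape cosmetic.
# SplineThicken sets strip width, while masked sine waves, vector offsets and camera offsets move vertices, with per-platform width and wave intensity.
# Vertex red varies wave and texture phase; green masks side offsets.
# Panned, noise-distorted smoke textures form depth-faded, edge-masked opacity, and emissive blends time-of-day gray with a curve-atlas color via UseGoldVersion.
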Material /BRCosmetics/Accessories/FORT_Capes/Cape_M_MED_MistMylar/FX/M_Cape_MistMylar_TrailTOP
graph {
 properties {
  bAutomaticallySetUsageInEditor=false
  bUsedWithMeshDeformer=true
  bUsedWithSkeletalMesh=true
  blendMode=BLEND_Translucent
  shadingModel=MSM_Unlit
  twoSided=true
  usedShadingModels=Unlit
 }
 nodes [
  n1 = TextureCoordinate()
  n2 = Multiply(A=OneMinus(None=ComponentMask(None=@n1) {r=true}))
  n3 = ScalarParameter() {parameterName=TopWidth}
  n4 = ScalarParameter() {defaultValue=0.3, parameterName=TopWidthLOD}
  n5 = Multiply(A=ComponentMask(None=TextureCoordinate()) {r=true}) {constB=10}
  n6 = Multiply(
   A=Subtract(A=Saturate(None=@n2)) {constB=0.6}
   B=LinearInterpolate(
    A=MaterialFunctionCall("Xb1,PS4"=@n3, "PC-High&&Epic,XSX,PS5"=@n3, "PC-Perf,Android"=@n4, Switch=@n4, "PC-Low(NotPerf)&&Medium"=@n3) {materialFunction=&PlatformSwitch_1D}.Result
    Alpha=Saturate(None=@n5)
   ) {constA=0.5}
  )
  n7 = MaterialFunctionCall(
   WidthBase=Multiply(A=@n6, B=Constant()) {constA=1}
   WidthTip=Multiply(A=@n6, B=Constant() {r=40}) {constA=1}
   Expand U or V UV Channel=StaticBool()
  ) {materialFunction=&SplineThicken}
  n8 = TextureCoordinate()
  n9 = LinearInterpolate(Alpha=VertexColor().R) {constA=0.8, constB=1.2}
  n10 = Power(Base=ComponentMask(None=Multiply(A=@n8, B=@n9)) {r=true}) {constExponent=0.5}
  n11 = Multiply(A=Add(A=Time(), B=VertexColor().R) {constB=0}) {constB=-0.5}
  n12 = LinearInterpolate(Alpha=Sine(None=Add(A=@n10, B=@n11)) {period=0.7}) {constA=-0.5, constB=0.5}
  n13 = Transform(None=VectorParameter() {parameterName=WPODirection, defaultValue=<rgba 0 0 1 1>}.RGB)
  n14 = VertexColor()
  n15 = ComponentMask(None=Multiply(A=TextureCoordinate(), B=LinearInterpolate(Alpha=@n14.R) {constA=0.8, constB=1.2})) {r=true}
  n16 = Add(A=Time(), B=VertexColor().R) {constB=0}
  n17 = Sine(None=Add(A=Power(Base=@n15) {constExponent=0.5}, B=Multiply(A=@n16) {constB=-0.5}) {constB=0}) {period=0.7}
  n18 = Transform(None=VectorParameter() {parameterName=WPODirection02, defaultValue=<rgba 0 0.6 0 1>}.RGB) {transformSourceType=TRANSFORMSOURCE_Local}
  n19 = Add(A=Multiply(A=Multiply(A=@n12, B=@n13)), B=Multiply(A=Multiply(A=@n17, B=@n18))) {constB=0}
  n20 = ScalarParameter() {defaultValue=7, parameterName=WPOIntensity}
  n21 = ScalarParameter() {defaultValue=5, parameterName=WPOIntensityLOD}
  n22 = MaterialFunctionCall("Xb1,PS4"=@n20, "PC-High&&Epic,XSX,PS5"=@n20, "PC-Perf,Android"=@n21, Switch=@n21, "PC-Low(NotPerf)&&Medium"=@n20) {materialFunction=&PlatformSwitch_1D}
  n23 = TextureCoordinate()
  n24 = Multiply(A=Power(Base=ComponentMask(None=@n23) {r=true}) {constExponent=1}, B=ScalarParameter() {defaultValue=3, parameterName=TopMask}) {constB=3}
  n25 = OneMinus(None=ComponentMask(None=TextureCoordinate()) {r=true})
  n26 = Multiply(A=LinearInterpolate(Alpha=Power(Base=@n25) {constExponent=0.5})) {constB=1.5}
  n27 = Multiply(A=Multiply(A=Saturate(None=@n24), B=Saturate(None=@n26))) {constB=1.3}
  n28 = Add(A=Multiply(A=Multiply(A=@n19, B=@n22.Result), B=Saturate(None=@n27))) {constB=0}
  n29 = Add(
   A=Add(A=Add(A=@n7.WorldPositionOffset, B=@n28) {constB=0}) {constB=0}
   B=Transform(None=VectorParameter() {parameterName=Offset02, defaultValue=<rgba 0 -2 2 1>}.RGB) {transformSourceType=TRANSFORMSOURCE_Local}
  )
  n30 = Multiply(
   A=Transform(None=VectorParameter() {parameterName=SideMeshOffset, defaultValue=<rgba 0 0 0 1>}.RGB) {transformSourceType=TRANSFORMSOURCE_Local}
   B=VertexColor().G
  )
  n31 = VertexColor()
  n32 = MaterialFunctionCall(Offset Amount=ScalarParameter() {defaultValue=3, parameterName="SideOffset Amount02 (S)"}) {materialFunction=/Engine/Functions/Engine_MaterialFunctions02/WorldPositionOffset/CameraOffset}
  Add(
   A=Add(A=Add(A=@n29, B=@n30), B=Multiply(A=@n31.G, B=@n32.Camera Offset))
   B=MaterialFunctionCall(Offset Amount=ScalarParameter() {defaultValue=7, parameterName="Offset Amount (S)"}) {materialFunction=&CameraOffset}.Camera Offset
  )
  Constant()
  n33 = Add(A=TextureCoordinate() {uTiling=2, vTiling=0.7}, B=AppendVector(A=VertexColor().R, B=Constant()))
  n34 = VertexColor()
  n35 = Constant()
  n36 = Panner(Coordinate=Add(A=TextureCoordinate() {uTiling=0.8}, B=AppendVector(A=@n34.R, B=@n35))) {speedX=0.1}
  n37 = Multiply(A=Subtract(A=TextureSample(UVs=@n36) {samplerType=SAMPLERTYPE_Grayscale, texture=&LowResBlurredNoise}.R) {constB=0.5}) {constB=0.2}
  n38 = TextureSample(UVs=Panner(Coordinate=Add(A=@n33, B=@n37) {constB=0}) {speedX=-1}) {samplerType=SAMPLERTYPE_Grayscale, texture=&T_Smoke_Trail_Soft}
  n39 = Add(A=TextureCoordinate() {uTiling=2, vTiling=0.7}, B=AppendVector(A=VertexColor().R, B=Constant()))
  n40 = TextureSample(UVs=Panner(Coordinate=Add(A=@n39, B=@n37) {constB=0}) {speedX=-0.5}) {samplerType=SAMPLERTYPE_Grayscale, texture=&LowResBlurredNoise}
  n41 = VertexColor()
  n42 = Constant()
  n43 = Panner(Coordinate=Add(A=TextureCoordinate() {uTiling=3}, B=AppendVector(A=@n41.R, B=@n42))) {speedX=-1}
  n44 = Saturate(None=Add(
    A=Add(A=@n38.R, B=@n40.R)
    B=TextureSample(UVs=@n43) {samplerType=SAMPLERTYPE_Grayscale, texture=&T_Smoke_Trail_Soft}.R
   ))
  n45 = Sine(None=ComponentMask(None=TextureCoordinate()) {g=true}) {period=2}
  n46 = OneMinus(None=ComponentMask(None=TextureCoordinate()) {r=true})
  n47 = Subtract(A=Saturate(None=Multiply(A=@n46))) {constB=0.7}
  n48 = Multiply(A=Multiply(A=Multiply(B=@n44) {constA=1}, B=Power(Base=@n45) {constExponent=1}), B=Saturate(None=Multiply(A=@n47) {constB=7}))
  n49 = TextureCoordinate()
  n50 = Multiply(A=Power(Base=ComponentMask(None=@n49) {r=true}) {constExponent=1.2}) {constB=30}
  DepthFade(
   Opacity=Multiply(A=Multiply(A=@n48) {constB=0.8}, B=Saturate(None=@n50))
   FadeDistance=ScalarParameter() {defaultValue=8, parameterName=FadeDistance}
  ) {fadeDistanceDefault=8}
  n51 = VectorParameter() {parameterName=GrayColor, defaultValue=<rgba 0.0063 0.0077 0.008 1>}
  n52 = VectorParameter() {parameterName=NightColor, defaultValue=<rgba 0.0035 0.0051 0.01 1>}
  n53 = CurveAtlasRowParameter(CurveTime=ComponentMask(None=TextureCoordinate()) {r=true}) {
   atlas=&/CAtlas_MistMylar_FX
   curve=&/CV_MistMylar_Color2
   defaultValue=1
   parameterName=Color
  }
  MaterialFunctionCall(Emissive=LinearInterpolate(
    A=MaterialFunctionCall(Color_Morning=@n51.RGB, Color_Day=@n51.RGB, Color_Evening=@n51.RGB, Color_Night=@n52.RGB) {materialFunction=&Color_Time_of_Day}.Result
    B=@n53
    Alpha=ScalarParameter() {parameterName=UseGoldVersion}
   )) {materialFunction=&MF_FX_EmissiveAdjustments}
  VectorParameter() {parameterName=GrayColor, defaultValue=<rgba 0.0063 0.0077 0.008 1>}
 ]
}
