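# Translucent, unlit, two-sided skeletal mesh material for the MistMylar cape FX, drawing scrolling smoky wisps.
# Opacity sums noise-distorted smoke and flame textures with caustics, shaped by UV tail and edge masks and depth fade.
# Color blends time-of-day gray with a curve-atlas gradient via UseGoldVersion, then passes through emissive adjustments.
# World position offset combines two U-driven sine waves, one phase-shifted by vertex red, with constant and camera offsets.
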Material /BRCosmetics/Accessories/FORT_Capes/Cape_M_MED_MistMylar/FX/M_MistMylar_Cape_BlandMesh
graph {
 properties {
  bAutomaticallySetUsageInEditor=false
  bUsedWithMeshDeformer=true
  bUsedWithSkeletalMesh=true
  blendMode=BLEND_Translucent
  shadingModel=MSM_Unlit
  twoSided=true
  usedShadingModels=Unlit
 }
 nodes [
  n1 = TextureCoordinate() {uTiling=0.5}
  n2 = Power(Base=ComponentMask(None=Multiply(A=@n1)) {r=true}) {constExponent=0.5}
  n3 = Multiply(A=Time()) {constB=-0.5}
  n4 = LinearInterpolate(Alpha=Sine(None=Add(A=@n2, B=@n3)) {period=0.7}) {constA=-0.5, constB=0.5}
  n5 = Transform(None=VectorParameter() {parameterName=WPODirection, defaultValue=<rgba 0 0 1 1>}.RGB)
  n6 = ComponentMask(None=Multiply(A=TextureCoordinate() {uTiling=0.5})) {r=true}
  n7 = Add(A=Time(), B=VertexColor().R) {constB=0}
  n8 = Sine(None=Add(A=Power(Base=@n6) {constExponent=0.5}, B=Multiply(A=@n7) {constB=-0.5}) {constB=0}) {period=0.7}
  n9 = Transform(None=VectorParameter() {parameterName=WPODirection02, defaultValue=<rgba 0 0.6 0 1>}.RGB) {transformSourceType=TRANSFORMSOURCE_Local}
  Add(A=Multiply(A=Multiply(A=@n4, B=@n5)), B=Multiply(A=Multiply(A=@n8, B=@n9))) {constB=0}
  n10 = VectorParameter() {parameterName=WPO Offset, defaultValue=<rgba 0 -1 -2.5 0>}
  n11 = ScalarParameter() {defaultValue=2, parameterName=WPOIntensity}
  n12 = ComponentMask(None=TextureCoordinate()) {r=true}
  n13 = Multiply(A=Multiply(A=Power(Base=@n12) {constExponent=0.6})) {constB=1.3}
  n14 = LinearInterpolate(B=Multiply(A=Multiply(B=@n11), B=Saturate(None=@n13)), Alpha=ScalarParameter() {defaultValue=1, parameterName=Use WPO})
  n15 = MaterialFunctionCall(Offset Amount=ScalarParameter() {defaultValue=3, parameterName="Offset Amount (S)"}) {materialFunction=/Engine/Functions/Engine_MaterialFunctions02/WorldPositionOffset/CameraOffset}
  Add(A=Add(A=Transform(None=@n10.RGB) {transformSourceType=TRANSFORMSOURCE_Local}, B=Add(A=@n14, B=@n15.Camera Offset) {constB=0})) {constB=0}
  Constant() {r=0.005}
  Constant()
  n16 = VectorParameter() {parameterName=GrayColor, defaultValue=<rgba 0.0063 0.0077 0.008 1>}
  n17 = MaterialFunctionCall(
   Color_Morning=@n16.RGB
   Color_Day=@n16.RGB
   Color_Evening=@n16.RGB
   Color_Night=VectorParameter() {parameterName=NightColor, defaultValue=<rgba 0.0035 0.0051 0.01 1>}.RGB
  ) {materialFunction=&Color_Time_of_Day}
  n18 = CurveAtlasRowParameter(CurveTime=ComponentMask(None=TextureCoordinate()) {r=true}) {
   atlas=&/CAtlas_MistMylar_FX
   curve=&/CV_MistMylar_Color3
   defaultValue=2
   parameterName=Color
  }
  n19 = ScalarParameter() {parameterName=UseGoldVersion}
  Reroute(None=MaterialFunctionCall(Emissive=LinearInterpolate(A=@n17.Result, B=@n18, Alpha=@n19)) {materialFunction=&MF_FX_EmissiveAdjustments}.Result)
  n20 = TextureCoordinate() {uTiling=0.5, vTiling=5}
  n21 = VertexColor()
  n22 = Constant()
  n23 = Panner(Coordinate=Add(A=TextureCoordinate() {uTiling=0.8}, B=AppendVector(A=@n21.R, B=@n22))) {speedX=0.1}
  n24 = Multiply(A=Subtract(A=TextureSample(UVs=@n23) {samplerType=SAMPLERTYPE_Grayscale, texture=&LowResBlurredNoise}.R) {constB=0.5}) {constB=0.2}
  n25 = TextureSample(UVs=Panner(Coordinate=Add(A=@n20, B=@n24) {constB=0}) {speedX=-0.7}) {
   samplerSource=SSM_Wrap_WorldGroupSettings
   samplerType=SAMPLERTYPE_Grayscale
   texture=&T_Smoke_Trail_Soft
  }
  n26 = Add(A=TextureCoordinate() {uTiling=0.5, vTiling=2}, B=@n24) {constB=0}
  n27 = Add(
   A=@n25.R
   B=TextureSample(UVs=Panner(Coordinate=@n26) {speedX=-0.5}) {samplerType=SAMPLERTYPE_LinearGrayscale, texture=/BRCosmetics/Weapons/FORT_Melee/Pickaxe_Cube_Ninja_Male/FX/T_FluidFlameTile}.R
  )
  n28 = TextureSample(UVs=Panner(Coordinate=TextureCoordinate() {uTiling=0.5, vTiling=2}) {speedX=-0.5}) {
   samplerSource=SSM_Wrap_WorldGroupSettings
   samplerType=SAMPLERTYPE_LinearGrayscale
   texture=/Game/Global/Textures/Default/Caustics/T_Nosie_Caustics_Grayscale
  }
  n29 = TextureCoordinate()
  n30 = Power(Base=OneMinus(None=ComponentMask(None=@n29) {r=true}), Exp=ScalarParameter() {defaultValue=1, parameterName=TailMask}) {constExponent=1}
  n31 = Multiply(A=Saturate(None=Add(A=@n27, B=@n28.R) {constB=0}), B=Saturate(None=Multiply(A=@n30)))
  n32 = TextureCoordinate()
  n33 = LinearInterpolate(Alpha=Power(Base=ComponentMask(None=@n32) {r=true}) {constExponent=1.5}) {constB=3}
  n34 = TextureCoordinate()
  n35 = Multiply(A=Power(Base=ComponentMask(None=@n34) {r=true}) {constExponent=1}) {constB=10}
  n36 = TextureCoordinate()
  n37 = Power(Base=Sine(None=ComponentMask(None=@n36) {g=true}) {period=2}, Exp=ScalarParameter() {defaultValue=1, parameterName=SideAlpha})
  n38 = ScalarParameter() {defaultValue=1.5, parameterName=EdgeFade}
  n39 = Multiply(A=Multiply(A=Power(Base=@n31, Exp=@n33), B=Saturate(None=@n35)), B=Saturate(None=Multiply(A=@n37, B=@n38) {constB=2}))
  n40 = ScalarParameter() {defaultValue=3, parameterName=FadeDistance}
  Saturate(None=Multiply(A=ScalarParameter() {defaultValue=1.2, parameterName=MainAlpha}, B=DepthFade(Opacity=@n39, FadeDistance=@n40) {fadeDistanceDefault=5}) {constA=1.3})
  VectorParameter() {parameterName=GrayColor, defaultValue=<rgba 0.0063 0.0077 0.008 1>}
 ]
}
