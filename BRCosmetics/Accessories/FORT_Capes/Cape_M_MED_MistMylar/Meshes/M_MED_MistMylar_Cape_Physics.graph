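# Physics asset for the MistMylar cape cosmetic (M_MED mesh), defining bodies and joints for the cape's dynamic bones.
# Six chains, three per side, each link a small MSTR root sphere to five to seven capsules through parent-child constraints.
# Swing limits loosen toward the tips, and every twist limit is set to zero.
# Unconstrained pelvis, thigh and calf capsules supply lower-body volumes, likely for cape collision.
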PhysicsAsset /BRCosmetics/Accessories/FORT_Capes/Cape_M_MED_MistMylar/Meshes/M_MED_MistMylar_Cape_Physics
bodies {
 calf_l [{shapeType=Capsule, length=17, radius=13.5, shapeName=calf_l_capsule, center=<-22.6397 -1.7794 0>, rotation=<rot 0 -90.0006 90>}]
 calf_r [{shapeType=Capsule, length=16.5, radius=13.5, shapeName=calf_r_capsule, center=<22.0193 2.0433 0>, rotation=<rot 0.0001 -90.0006 90>}]
 dyn_cape_l_a_01 [{shapeType=Capsule, length=5, radius=4, shapeName=dyn_cape_l_a_01_capsule, center=<8 0 0>, rotation=<rot -90 0 0>}]
 dyn_cape_l_a_02 [{shapeType=Capsule, length=5, radius=4, shapeName=dyn_cape_l_a_02_capsule, center=<8 0 0>, rotation=<rot -90 0 0>}]
 dyn_cape_l_a_03 [{shapeType=Capsule, length=5, radius=4, shapeName=dyn_cape_l_a_03_capsule, center=<8 0 0>, rotation=<rot -90 0 0>}]
 dyn_cape_l_a_04 [{shapeType=Capsule, length=5, radius=4, shapeName=dyn_cape_l_a_04_capsule, center=<8 0 0>, rotation=<rot -90 0 0>}]
 dyn_cape_l_a_05 [{shapeType=Capsule, length=5, radius=4, shapeName=dyn_cape_l_a_05_capsule, center=<8 0 0>, rotation=<rot -90 0 0>}]
 dyn_cape_l_a_06 [{shapeType=Capsule, length=5, radius=4, shapeName=dyn_cape_l_a_06_capsule, center=<8 0 0>, rotation=<rot -90 0 0>}]
 dyn_cape_l_a_07 [{shapeType=Capsule, length=5, radius=4, shapeName=dyn_cape_l_a_07_capsule, center=<8 0 0>, rotation=<rot -90 0 0>}]
 dyn_cape_l_a_MSTR [{shapeType=Sphere, radius=0.505, shapeName=dyn_cape_l_a_MSTR_sphere}]
 dyn_cape_l_b_01 [{shapeType=Capsule, length=5, radius=4, shapeName=dyn_cape_l_b_01_capsule, center=<10 0 0>, rotation=<rot -90 0 0>}]
 dyn_cape_l_b_02 [{shapeType=Capsule, length=5, radius=4, shapeName=dyn_cape_l_b_02_capsule, center=<10 0 0>, rotation=<rot -90 0 0>}]
 dyn_cape_l_b_03 [{shapeType=Capsule, length=5, radius=4, shapeName=dyn_cape_l_b_03_capsule, center=<10 0 0>, rotation=<rot -90 0 0>}]
 dyn_cape_l_b_04 [{shapeType=Capsule, length=5, radius=4, shapeName=dyn_cape_l_b_04_capsule, center=<10 0 0>, rotation=<rot -90 0 0>}]
 dyn_cape_l_b_05 [{shapeType=Capsule, length=5, radius=4, shapeName=dyn_cape_l_b_05_capsule, center=<10 0 0>, rotation=<rot -90 0 0>}]
 dyn_cape_l_b_06 [{shapeType=Capsule, length=5, radius=4, shapeName=dyn_cape_l_b_06_capsule, center=<10 0 0>, rotation=<rot -90 0 0>}]
 dyn_cape_l_b_MSTR [{shapeType=Sphere, radius=0.505, shapeName=dyn_cape_l_b_MSTR_sphere}]
 dyn_cape_l_c_01 [{shapeType=Capsule, length=5, radius=4, shapeName=dyn_cape_l_c_01_capsule, center=<5 0 0>, rotation=<rot -90 0 0>}]
 dyn_cape_l_c_02 [{shapeType=Capsule, length=5, radius=4, shapeName=dyn_cape_l_c_02_capsule, center=<5 0 0>, rotation=<rot -90 0 0>}]
 dyn_cape_l_c_03 [{shapeType=Capsule, length=5, radius=4, shapeName=dyn_cape_l_c_03_capsule, center=<5 0 0>, rotation=<rot -90 0 0>}]
 dyn_cape_l_c_04 [{shapeType=Capsule, length=5, radius=4, shapeName=dyn_cape_l_c_04_capsule, center=<5 0 0>, rotation=<rot -90 0 0>}]
 dyn_cape_l_c_05 [{shapeType=Capsule, length=5, radius=4, shapeName=dyn_cape_l_c_05_capsule, center=<5 0 0>, rotation=<rot -90 0 0>}]
 dyn_cape_l_c_MSTR [{shapeType=Sphere, radius=0.505, shapeName=dyn_cape_l_c_MSTR_sphere}]
 dyn_cape_r_a_01 [{shapeType=Capsule, length=5, radius=4, shapeName=dyn_cape_r_a_01_capsule, center=<10 0 0>, rotation=<rot -90 0 0>}]
 dyn_cape_r_a_02 [{shapeType=Capsule, length=5, radius=4, shapeName=dyn_cape_r_a_02_capsule, center=<10 0 0>, rotation=<rot -90 0 0>}]
 dyn_cape_r_a_03 [{shapeType=Capsule, length=5, radius=4, shapeName=dyn_cape_r_a_03_capsule, center=<10 0 0>, rotation=<rot -90 0 0>}]
 dyn_cape_r_a_04 [{shapeType=Capsule, length=5, radius=4, shapeName=dyn_cape_r_a_04_capsule, center=<10 0 0>, rotation=<rot -90 0 0>}]
 dyn_cape_r_a_05 [{shapeType=Capsule, length=5, radius=4, shapeName=dyn_cape_r_a_05_capsule, center=<10 0 0>, rotation=<rot -90 0 0>}]
 dyn_cape_r_a_06 [{shapeType=Capsule, length=5, radius=4, shapeName=dyn_cape_r_a_06_capsule, center=<10 0 0>, rotation=<rot -90 180 180>}]
 dyn_cape_r_a_07 [{shapeType=Capsule, length=5, radius=4, shapeName=dyn_cape_r_a_07_capsule, center=<10 0 0>, rotation=<rot -90 0 0>}]
 dyn_cape_r_a_MSTR [{shapeType=Sphere, radius=0.505, shapeName=dyn_cape_r_a_MSTR_sphere}]
 dyn_cape_r_b_01 [{shapeType=Capsule, length=5, radius=4, shapeName=dyn_cape_r_b_01_capsule, center=<8 0 0>, rotation=<rot -90 0 0>}]
 dyn_cape_r_b_02 [{shapeType=Capsule, length=5, radius=4, shapeName=dyn_cape_r_b_02_capsule, center=<8 0 0>, rotation=<rot -90 0 0>}]
 dyn_cape_r_b_03 [{shapeType=Capsule, length=5, radius=4, shapeName=dyn_cape_r_b_03_capsule, center=<8 0 0>, rotation=<rot -90 0 0>}]
 dyn_cape_r_b_04 [{shapeType=Capsule, length=5, radius=4, shapeName=dyn_cape_r_b_04_capsule, center=<8 0 0>, rotation=<rot -90 0 0>}]
 dyn_cape_r_b_05 [{shapeType=Capsule, length=5, radius=4, shapeName=dyn_cape_r_b_05_capsule, center=<8 0 0>, rotation=<rot -90 0 0>}]
 dyn_cape_r_b_06 [{shapeType=Capsule, length=5, radius=4, shapeName=dyn_cape_r_b_06_capsule, center=<8 0 0>, rotation=<rot -90 180 180>}]
 dyn_cape_r_b_MSTR [{shapeType=Sphere, radius=0.505, shapeName=dyn_cape_r_b_MSTR_sphere}]
 dyn_cape_r_c_01 [{shapeType=Capsule, length=5, radius=4, shapeName=dyn_cape_r_c_01_capsule, center=<7 0 0>, rotation=<rot -90 0 0>}]
 dyn_cape_r_c_02 [{shapeType=Capsule, length=5, radius=4, shapeName=dyn_cape_r_c_02_capsule, center=<7 0 0>, rotation=<rot -90 0 0>}]
 dyn_cape_r_c_03 [{shapeType=Capsule, length=5, radius=4, shapeName=dyn_cape_r_c_03_capsule, center=<7 0 0>, rotation=<rot -90 0 0>}]
 dyn_cape_r_c_04 [{shapeType=Capsule, length=5, radius=4, shapeName=dyn_cape_r_c_04_capsule, center=<7 0 0>, rotation=<rot -90 0 0>}]
 dyn_cape_r_c_05 [{shapeType=Capsule, length=5, radius=4, shapeName=dyn_cape_r_c_05_capsule, center=<7 0 0>, rotation=<rot -90 0 0>}]
 dyn_cape_r_c_06 [{shapeType=Capsule, length=5, radius=4, shapeName=dyn_cape_r_c_06_capsule, center=<7 0 0>, rotation=<rot -90 180 180>}]
 dyn_cape_r_c_MSTR [{shapeType=Sphere, radius=0.505, shapeName=dyn_cape_r_c_MSTR_sphere}]
 pelvis [{shapeType=Capsule, length=13.005, radius=14.505, shapeName=pelvis_capsule, center=<-4.9339 -2.0646 0>, rotation=<rot 0 -90 0>}]
 thigh_l [{shapeType=Capsule, length=14.5, radius=11.5, shapeName=thigh_l_capsule, center=<-23.8494 0.0002 0>, rotation=<rot 0.0001 -90.0006 -90>}]
 thigh_r [{shapeType=Capsule, length=13.5, radius=12.5, shapeName=thigh_r_capsule, center=<23.5984 -0.0002 0>, rotation=<rot 0.0001 -90.0006 -90>}]
}
constraints [
 {bone1Name=dyn_cape_l_a_01, bone2Name=dyn_cape_l_a_MSTR, swing1LimitDegrees=15, swing2LimitDegrees=15, twistLimitDegrees=0}
 {bone1Name=dyn_cape_l_a_02, bone2Name=dyn_cape_l_a_01, swing1LimitDegrees=15, swing2LimitDegrees=15, twistLimitDegrees=0}
 {bone1Name=dyn_cape_l_a_03, bone2Name=dyn_cape_l_a_02, swing1LimitDegrees=15, swing2LimitDegrees=15, twistLimitDegrees=0}
 {bone1Name=dyn_cape_l_a_04, bone2Name=dyn_cape_l_a_03, swing1LimitDegrees=15, swing2LimitDegrees=15, twistLimitDegrees=0}
 {bone1Name=dyn_cape_l_a_05, bone2Name=dyn_cape_l_a_04, swing1LimitDegrees=20, swing2LimitDegrees=20, twistLimitDegrees=0}
 {bone1Name=dyn_cape_l_a_06, bone2Name=dyn_cape_l_a_05, swing1LimitDegrees=20, swing2LimitDegrees=20, twistLimitDegrees=0}
 {bone1Name=dyn_cape_l_a_07, bone2Name=dyn_cape_l_a_06, swing1LimitDegrees=20, swing2LimitDegrees=20, twistLimitDegrees=0}
 {bone1Name=dyn_cape_r_c_02, bone2Name=dyn_cape_r_c_01, swing1LimitDegrees=15, swing2LimitDegrees=15, twistLimitDegrees=0}
 {bone1Name=dyn_cape_r_c_03, bone2Name=dyn_cape_r_c_02, swing1LimitDegrees=15, swing2LimitDegrees=15, twistLimitDegrees=0}
 {bone1Name=dyn_cape_r_c_04, bone2Name=dyn_cape_r_c_03, swing1LimitDegrees=20, swing2LimitDegrees=20, twistLimitDegrees=0}
 {bone1Name=dyn_cape_r_c_05, bone2Name=dyn_cape_r_c_04, swing1LimitDegrees=20, swing2LimitDegrees=20, twistLimitDegrees=0}
 {bone1Name=dyn_cape_r_c_06, bone2Name=dyn_cape_r_c_05, swing1LimitDegrees=20, swing2LimitDegrees=20, twistLimitDegrees=0}
 {bone1Name=dyn_cape_r_b_01, bone2Name=dyn_cape_r_b_MSTR, swing1LimitDegrees=15, swing2LimitDegrees=15, twistLimitDegrees=0}
 {bone1Name=dyn_cape_r_b_02, bone2Name=dyn_cape_r_b_01, swing1LimitDegrees=15, swing2LimitDegrees=15, twistLimitDegrees=0}
 {bone1Name=dyn_cape_r_b_03, bone2Name=dyn_cape_r_b_02, swing1LimitDegrees=15, swing2LimitDegrees=15, twistLimitDegrees=0}
 {bone1Name=dyn_cape_r_b_04, bone2Name=dyn_cape_r_b_03, swing1LimitDegrees=20, swing2LimitDegrees=20, twistLimitDegrees=0}
 {bone1Name=dyn_cape_r_b_05, bone2Name=dyn_cape_r_b_04, swing1LimitDegrees=20, swing2LimitDegrees=20, twistLimitDegrees=0}
 {bone1Name=dyn_cape_r_b_06, bone2Name=dyn_cape_r_b_05, swing1LimitDegrees=20, swing2LimitDegrees=20, twistLimitDegrees=0}
 {bone1Name=dyn_cape_r_c_01, bone2Name=dyn_cape_r_c_MSTR, swing1LimitDegrees=15, swing2LimitDegrees=15, twistLimitDegrees=0}
 {bone1Name=dyn_cape_r_a_01, bone2Name=dyn_cape_r_a_MSTR, swing1LimitDegrees=15, swing2LimitDegrees=15, twistLimitDegrees=0}
 {bone1Name=dyn_cape_r_a_02, bone2Name=dyn_cape_r_a_01, swing1LimitDegrees=15, swing2LimitDegrees=15, twistLimitDegrees=0}
 {bone1Name=dyn_cape_r_a_03, bone2Name=dyn_cape_r_a_02, swing1LimitDegrees=15, swing2LimitDegrees=15, twistLimitDegrees=0}
 {bone1Name=dyn_cape_r_a_04, bone2Name=dyn_cape_r_a_03, swing1LimitDegrees=15, swing2LimitDegrees=15, twistLimitDegrees=0}
 {bone1Name=dyn_cape_r_a_05, bone2Name=dyn_cape_r_a_04, swing1LimitDegrees=20, swing2LimitDegrees=20, twistLimitDegrees=0}
 {bone1Name=dyn_cape_r_a_06, bone2Name=dyn_cape_r_a_05, swing1LimitDegrees=20, swing2LimitDegrees=20, twistLimitDegrees=0}
 {bone1Name=dyn_cape_r_a_07, bone2Name=dyn_cape_r_a_06, swing1LimitDegrees=20, swing2LimitDegrees=20, twistLimitDegrees=0}
 {bone1Name=dyn_cape_l_c_01, bone2Name=dyn_cape_l_c_MSTR, swing1LimitDegrees=15, swing2LimitDegrees=15, twistLimitDegrees=0}
 {bone1Name=dyn_cape_l_c_02, bone2Name=dyn_cape_l_c_01, swing1LimitDegrees=15, swing2LimitDegrees=15, twistLimitDegrees=0}
 {bone1Name=dyn_cape_l_c_03, bone2Name=dyn_cape_l_c_02, swing1LimitDegrees=20, swing2LimitDegrees=20, twistLimitDegrees=0}
 {bone1Name=dyn_cape_l_c_04, bone2Name=dyn_cape_l_c_03, swing1LimitDegrees=20, swing2LimitDegrees=20, twistLimitDegrees=0}
 {bone1Name=dyn_cape_l_c_05, bone2Name=dyn_cape_l_c_04, swing1LimitDegrees=20, swing2LimitDegrees=20, twistLimitDegrees=0}
 {bone1Name=dyn_cape_l_b_01, bone2Name=dyn_cape_l_b_MSTR, swing1LimitDegrees=15, swing2LimitDegrees=15, twistLimitDegrees=0}
 {bone1Name=dyn_cape_l_b_02, bone2Name=dyn_cape_l_b_01, swing1LimitDegrees=15, swing2LimitDegrees=15, twistLimitDegrees=0}
 {bone1Name=dyn_cape_l_b_03, bone2Name=dyn_cape_l_b_02, swing1LimitDegrees=15, swing2LimitDegrees=15, twistLimitDegrees=0}
 {bone1Name=dyn_cape_l_b_04, bone2Name=dyn_cape_l_b_03, swing1LimitDegrees=20, swing2LimitDegrees=20, twistLimitDegrees=0}
 {bone1Name=dyn_cape_l_b_05, bone2Name=dyn_cape_l_b_04, swing1LimitDegrees=20, swing2LimitDegrees=20, twistLimitDegrees=0}
 {bone1Name=dyn_cape_l_b_06, bone2Name=dyn_cape_l_b_05, swing1LimitDegrees=20, swing2LimitDegrees=20, twistLimitDegrees=0}
]
